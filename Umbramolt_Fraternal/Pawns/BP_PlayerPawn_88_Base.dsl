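; Base player pawn Blueprint for an "88" henchman in Umbramolt Fraternal, tied to the Yuki and Gogo twin pawns and a spawning portal.
; On BeginPlay the authority fires a spawn GameplayCue, and server death notifies the portal through the UmbraMoltFraternalInterface.
; Passed-in twin references get health-change bindings that drive force-elimination of this pawn, which also happens if neither twin is valid.
; Intro speech, walkout emote and twin-state getters return empty or false defaults.
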(event EventBeginPlay
  (|CallParentFunction)
  (Networking|SwitchHasAuthority
    (:Authority
      (GameplayCue|ExecuteGameplayCueOnActor(Burst) self (Variables|Default|GetSpawn_GCN) (Ability|GameplayCue|MakeGameplayCueParameters 0.0)))))

(event EventOnDeathServer
  (params (Damage :type Float%20%28single-precision%29) (DamageTags :type Gameplay%20Tag%20Container%20Structure%20%28by%20ref%29) (Momentum :type Vector :default "0, 0, 0") (HitInfo :type Hit%20Result%20Structure%20%28by%20ref%29) (InstigatedBy :type Controller%20Object%20Reference) (DamageCauser :type Actor%20Object%20Reference) (EffectContext :type Gameplay%20Effect%20Context%20Handle%20Structure))
  (bind _portalref (Variables|Default|GetPortalRef))
  (|CallParentFunction Damage DamageTags Momentum HitInfo InstigatedBy DamageCauser EffectContext)
  (Utilities|IsValid _portalref
    (:Is%20Valid
      (Class|UmbraMoltFraternalInterface|SpawnedHenchmenEliminated(Message) _portalref))))

(event Custom|ForceEliminate
  (FortPawn|ForceKill))

(event EventPassBothTwinsReferences
  (params (YukiPawn :type Fort%20Pawn%20Object%20Reference) (GogoPawn :type Fort%20Pawn%20Object%20Reference))
  (Utilities|IsValid YukiPawn
    (:Is%20Valid
      (bind _output_get (Variables|Default|SetYukiRef YukiPawn))
      (Health|BindEventtoOnPawnHealthChanged _output_get (EventDispatchers|CreateEvent :!event Custom|YukiHealthChanged)))
    (:Is%20Not%20Valid
      _))
  (Utilities|IsValid GogoPawn
    (:Is%20Valid
      (bind _output_get_2 (Variables|Default|SetGogoRef GogoPawn))
      (Health|BindEventtoOnPawnHealthChanged _output_get_2 (EventDispatchers|CreateEvent :!event Custom|GogoHealthChanged)))
    (:Is%20Not%20Valid
      _))
  (if (not (or (Utilities|IsValid (Variables|Default|GetYukiRef) :!class /Script/Engine.KismetSystemLibrary) (Utilities|IsValid (Variables|Default|GetGogoRef) :!class /Script/Engine.KismetSystemLibrary)))
    (CallFunction|ForceEliminate)))

(event Custom|YukiHealthChanged
  (bind _gogoref (Variables|Default|GetGogoRef))
  (if (<= (FortPawn|GetHealth (Variables|Default|GetYukiRef)) 0)
    (Utilities|IsValid _gogoref
      (:Is%20Valid
        (if (<= (FortPawn|GetHealth _gogoref) 0)
          _))
      (:Is%20Not%20Valid
        _))
    (CallFunction|ForceEliminate)))

(event Custom|GogoHealthChanged
  (bind _yukiref (Variables|Default|GetYukiRef))
  (if (<= (FortPawn|GetHealth (Variables|Default|GetGogoRef)) 0)
    (Utilities|IsValid _yukiref
      (:Is%20Valid
        (if (<= (FortPawn|GetHealth _yukiref) 0)
          _))
      (:Is%20Not%20Valid
        _))
    (CallFunction|ForceEliminate)))

(event EventPassPortalRef
  (params (Actor :type Actor%20Object%20Reference))
  (Variables|Default|SetPortalRef Actor))

(fn GetIntroSpeechTag
  (return 0))

(fn GetPostWalkoutEmotes
  (return 0))

(fn GetTalkAtIntro
  (return false))

(fn GetTwinState
  (return false))

(fn GetWalkoutEmoteToPlay
  (return 0))

(fn ConstructionScript
  (|CallParentFunction))
