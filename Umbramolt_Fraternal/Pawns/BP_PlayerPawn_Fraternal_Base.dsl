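; Base pawn for one of the Umbramolt Fraternal twin bosses, handling walkout intro, shared twin state and elimination.
; Damage is scaled by a hotfix value during walkout; taking a hit tags both twins walkout-complete.
; After walkout, a retriggerable delay on Unaware alert force-kills this pawn if its twin is Unaware or missing.
; Twin death triggers speech, and the IsTwinActive RepNotify toggles visibility, collision and hidden/storm-immunity tags.
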(event EventBeginPlay
  (|CallParentFunction)
  (Ability|Tasks|WaitGameplayTagAddToActor 0 "(TagName=\"Umbramolt_Fraternal.WalkoutComplete\")"
    (:Added
      (bind _asbp_bot_controller_boss_base (Utilities|Casting|CastToBP_BotController_Boss_Base (Pawn|GetController)))
      (AlertLevel|BindEventtoOnAlertLevelChangedEvent _asbp_bot_controller_boss_base (EventDispatchers|CreateEvent :!event Custom|FraternalAlertLevelChanged)))))

(event EventToggleTwinState
  (params (Active :type Boolean))
  (Variables|Default|SetIsTwinActive Active))

(event Custom|FraternalAlertLevelChanged
  (params (BotController :type Fort%20Athena%20AIBot%20Controller%20Object%20Reference) (OldAlertLevel :type EAlertLevel%20Enum :default "Unaware") (NewAlertLevel :type EAlertLevel%20Enum :default "Unaware"))
  (bind _twinref (Variables|Default|GetTwinRef))
  (if (Utilities|Enum|Equal(Enum) NewAlertLevel "Unaware")
    (Utilities|FlowControl|RetriggerableDelay (ScalableFloat|GetValueatLevel (Variables|Default|GetHFUnawareCheckFrequency)))
    (Utilities|IsValid _twinref
      (:Is%20Valid
        (bind _asbp_bot_controller_boss_base (Utilities|Casting|CastToBP_BotController_Boss_Base (Pawn|GetController _twinref))
          (:then
            (if (Utilities|Enum|Equal(Enum) (Class|FortAthenaAIBotController|GetAlertLevel _asbp_bot_controller_boss_base) "Unaware")
              (CallFunction|ForceEliminate)))
          (:CastFailed
            _)))
      (:Is%20Not%20Valid
        _))
    (CallFunction|ForceEliminate)))

(event EventSetWalkoutEmoteToPlay
  (params (Index :type Integer))
  (Variables|Default|SetWalkoutEmoteToPlay Index))

(event EventSetIsDoingWalkout
  (params (Bool :type Boolean))
  (Variables|Default|SetIsDoingWalkout Bool))

(event EventPassTwinReference
  (params (TwinPawn :type Fort%20Pawn%20Object%20Reference))
  (bind _output_get (Variables|Default|SetTwinRef TwinPawn))
  (Utilities|IsValid _output_get
    (:Is%20Valid
      (EventDispatchers|BindEventtoOnDied (Variables|Default|GetTwinRef) (EventDispatchers|CreateEvent :!event Custom|TwinDied)))))

(event Custom|TwinDied
  (params (DamagedActor :type Actor%20Object%20Reference) (Damage :type Float%20%28single-precision%29) (InstigatedBy :type Controller%20Object%20Reference) (DamageCauser :type Actor%20Object%20Reference) (HitLocation :type Vector) (FHitComponent :type Primitive%20Component%20Object%20Reference) (BoneName :type Name) (Momentum :type Vector))
  (bind _twinref (Variables|Default|GetTwinRef))
  (Class|BPFLAICommonFunctions|PlaySpeechBubble self (Variables|SpeechTags|GetTwinEliminated_SpeechTag) true true)
  (Variables|Default|SetTwinEliminated true)
  (Utilities|IsValid _twinref
    (:Is%20Valid
      (EventDispatchers|UnbindEventfromOnDied _twinref (EventDispatchers|CreateEvent :!event Custom|TwinDied)))))

(event EventOnDamageServer
  (params (Damage :type Float%20%28single-precision%29) (DamageTags :type Gameplay%20Tag%20Container%20Structure%20%28by%20ref%29) (Momentum :type Vector :default "0, 0, 0") (HitInfo :type Hit%20Result%20Structure%20%28by%20ref%29) (InstigatedBy :type Controller%20Object%20Reference) (DamageCauser :type Actor%20Object%20Reference) (EffectContext :type Gameplay%20Effect%20Context%20Handle%20Structure))
  (bind _twinref (Variables|Default|GetTwinRef))
  (|CallParentFunction Damage DamageTags Momentum HitInfo InstigatedBy DamageCauser EffectContext)
  (Class|BPFLAICommonFunctions|AddUniqueGameplayTag self "(TagName=\"Umbramolt_Fraternal.WalkoutComplete\")")
  (Utilities|IsValid _twinref
    (:Is%20Valid
      (Class|BPFLAICommonFunctions|AddUniqueGameplayTag _twinref "(TagName=\"Umbramolt_Fraternal.WalkoutComplete\")"))))

(event Custom|ForceEliminate
  (FortPawn|ForceKill))

(event EventOnDeathServer
  (params (Damage :type Float%20%28single-precision%29) (DamageTags :type Gameplay%20Tag%20Container%20Structure%20%28by%20ref%29) (Momentum :type Vector :default "0, 0, 0") (HitInfo :type Hit%20Result%20Structure%20%28by%20ref%29) (InstigatedBy :type Controller%20Object%20Reference) (DamageCauser :type Actor%20Object%20Reference) (EffectContext :type Gameplay%20Effect%20Context%20Handle%20Structure))
  (|CallParentFunction Damage DamageTags Momentum HitInfo InstigatedBy DamageCauser EffectContext)
  (if (Variables|Default|GetTwinEliminated)
    (Class|BPFLAICommonFunctions|PlaySpeechBubble self (Variables|SpeechTags|GetEliminated_SpeechTag) true true)))

(event EventSetTalkAtIntro
  (params (Bool :type Boolean))
  (Variables|Default|SetTalkAtIntro Bool))

(fn BlueprintModifyIncomingDamage
  (params (Damage :type Float%20%28single-precision%29) (DamageTags :type Gameplay%20Tag%20Container%20Structure%20%28by%20ref%29) (EffectContext :type Gameplay%20Effect%20Context%20Handle%20Structure%20%28by%20ref%29) (EventInstigator :type Controller%20Object%20Reference) (DamageCauser :type Actor%20Object%20Reference) (HitInfo :type Hit%20Result%20Structure))
  (bind _returnvalue (|CallParentFunction Damage DamageTags EffectContext EventInstigator DamageCauser HitInfo))
  (return (* _returnvalue (Math|Float|SelectFloat (ScalableFloat|GetValueatLevel (Variables|Hotfix|GetWalkoutDmgMod_Hf)) 1.0 (and (GameplayTags|HasMatchingGameplayTag "(TagName=\"Umbramolt_Fraternal.StartWalkout\")") (not (GameplayTags|HasMatchingGameplayTag "(TagName=\"Umbramolt_Fraternal.WalkoutComplete\")")))))))

(fn GetAdditiveHitReactsMontage
  (return (select (Variables|Default|GetIsDoingWalkout) 0 "/Game/Animation/Game/MainPlayer/Montages/General/AdditiveHitReacts.AdditiveHitReacts")))

(fn GetIntroSpeechTag
  (return (Variables|SpeechTags|GetIntro_SpeechTag)))

(fn GetPostWalkoutEmotes
  (return (Variables|Default|GetPostWalkoutEmotes)))

(fn GetTalkAtIntro
  (return (Variables|Default|GetTalkAtIntro)))

(fn GetTwinState
  (return (Variables|Default|GetIsTwinActive)))

(fn GetWalkoutEmoteToPlay
  (return (Variables|Default|GetWalkoutEmoteToPlay)))

(fn OnRep_IsTwinActive
  (bind _self self)
  (bind _self_2 self)
  (bind _istwinactive (Variables|Default|GetIsTwinActive))
  (Rendering|SetActorHiddenInGame (not _istwinactive))
  (Collision|SetCollisionProfileName (Variables|Character|GetCapsuleComponent) (Utilities|String|SelectName "FortPawnCapsule" "FortAIPawnOutsideTheLoop" _istwinactive) true :!class /Script/Engine.PrimitiveComponent)
  (Collision|SetCollisionProfileName (Variables|Character|GetMesh) (Utilities|String|SelectName "FortPawnMesh" "FortAIPawnOutsideTheLoop" _istwinactive) true :!class /Script/Engine.PrimitiveComponent)
  (if (Variables|Default|GetIsTwinActive)
    (Class|BPFLAICommonFunctions|RemoveAllInstancesOfGameplayTag _self_2 "(TagName=\"Athena.AI.Hidden\")")
    (Class|BPFLAICommonFunctions|RemoveAllInstancesOfGameplayTag _self "(TagName=\"Gameplay.Damage.Environment.IgnoreStormDamage\")")
    (else
      (Class|BPFLAICommonFunctions|AddUniqueGameplayTag _self_2 "(TagName=\"Athena.AI.Hidden\")")
      (Class|BPFLAICommonFunctions|AddUniqueGameplayTag _self "(TagName=\"Gameplay.Damage.Environment.IgnoreStormDamage\")")
      (Variables|Default|SetUseAlertness false)
      (Variables|Default|SetUseAlertSound false))))

(fn ConstructionScript
  (|CallParentFunction))
